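# Looping outdoor sand and wind ambience cue for the Desert Wasteland biome, routed to exterior ambient bus and reverb.
# The first two quality levels mix a random wind-sand emitter loop with a half-volume random quad swirl.
# The other two randomly pick one stereo swirl or emitter loop; stereo swirl 02 is not set to loop.
# Both branches use nested, endlessly looping envelopes to slowly modulate pitch and volume.
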SoundCue /Game/Sounds/Helios/Biomes/DesertWasteland/Ambience_DesertWasteland_Sand_Loop_Cue
root=QualityLevel_0
nodes [
 {
  id=Enveloper_0
  children [Random_1]
  properties {
   bLoop=true
   bLoopIndefinitely=true
   loopEnd=30
   pitchCurve {editorCurveData {keys [{interpMode=RCIM_Cubic, time=0}, <at 15.9468 0.9181 RCIM_Cubic>, {interpMode=RCIM_Cubic, time=30}]}}
   volumeCurve {editorCurveData {keys [<at 0 1>]}}
  }
 }
 {
  id=Enveloper_1
  children [Mixer_0]
  properties {
   bLoop=true
   bLoopIndefinitely=true
   loopEnd=30
   pitchCurve {editorCurveData {keys [{interpMode=RCIM_Cubic, time=0}, <at 15.9468 0.9181 RCIM_Cubic>, {interpMode=RCIM_Cubic, time=30}]}}
   volumeCurve {editorCurveData {keys [<at 0 1>]}}
  }
 }
 {
  id=Enveloper_2
  children [Enveloper_1]
  properties {
   bLoop=true
   bLoopIndefinitely=true
   loopEnd=40
   pitchCurve {editorCurveData {keys [{time=0}]}}
   volumeCurve {editorCurveData {keys [
      {arriveTangent=-0.01, interpMode=RCIM_Cubic, time=0, value=1}
      <at 20 0.8 RCIM_Cubic>
      {interpMode=RCIM_Cubic, leaveTangent=0.01, time=40, value=1}
     ]}}
  }
 }
 {
  id=Enveloper_3
  children [Enveloper_0]
  properties {
   bLoop=true
   bLoopIndefinitely=true
   loopEnd=30
   pitchCurve {editorCurveData {keys [{time=0}]}}
   volumeCurve {editorCurveData {keys [
      {arriveTangent=-0.01, interpMode=RCIM_Cubic, time=0, value=1}
      <at 10 0.7 RCIM_Cubic>
      {interpMode=RCIM_Cubic, leaveTangent=0.01, time=20, value=1}
     ]}}
  }
 }
 {id=Mixer_0, children [Random_2,Random_0], properties {inputVolume [1,0.5]}}
 {id=QualityLevel_0, children [Enveloper_2,Enveloper_2,Enveloper_3,Enveloper_3]}
 {id=Random_0, children [WavePlayer_0,WavePlayer_1], properties {weights [1,1]}}
 {id=Random_1, children [WavePlayer_3,WavePlayer_2,Random_2], properties {weights [1,1,1]}}
 {id=Random_2, children [WavePlayer_6,WavePlayer_7], properties {weights [1,1]}}
 {
  id=WavePlayer_0
  properties {bLooping=true, soundWaveAssetPtr=&/Ambience_Sand_Only_Swirling_Texture_Loop_02_Quad}
 }
 {
  id=WavePlayer_1
  properties {bLooping=true, soundWaveAssetPtr=&/Ambience_Sand_Only_Swirling_Texture_Loop_01_Quad}
 }
 {
  id=WavePlayer_2
  properties {soundWaveAssetPtr=&/Ambience_Sand_Only_Swirling_Texture_Loop_02_Stereo}
 }
 {
  id=WavePlayer_3
  properties {bLooping=true, soundWaveAssetPtr=&/Ambience_Sand_Only_Swirling_Texture_Loop_01_Stereo}
 }
 {
  id=WavePlayer_6
  properties {bLooping=true, soundWaveAssetPtr=/Game/Sounds/Helios/Biomes/DesertWasteland/Spots/Spot_Desert_Wind_Sand_Emitter_Loop_02}
 }
 {
  id=WavePlayer_7
  properties {bLooping=true, soundWaveAssetPtr=/Game/Sounds/Helios/Biomes/DesertWasteland/Spots/Spot_Desert_Wind_Sand_Emitter_Loop_01}
 }
]
properties {
 attenuationSettings=&Fort_Ambient_Athena_Att
 maxDistance=2097152
 soundClassObject=&Fort_Ambient_Outdoors_Athena
 volumeMultiplier=0.15
 busSends [{audioBus=null, minSendLevel=1, sendLevel=1, soundSourceBus=&AmbientAudio_Exterior_SourceBus}]
 soundSubmixSends [{minSendLevel=1, soundSubmix=&Fortnite_Reverb_Default}]
}
